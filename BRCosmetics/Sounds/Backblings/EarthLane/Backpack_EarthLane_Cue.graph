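# Sound cue for the EarthLane back bling, holding its elimination-reaction sounds.
# A Random node picks one of five equally weighted ElimReact waves and feeds a Modulator that slightly varies volume at fixed pitch.
# It uses the close-range pet attenuation, the Backblings sound class and a reactive back bling property sheet.
# A concurrency override limits it to one instance, stopping the oldest.
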SoundCue /BRCosmetics/Sounds/Backblings/EarthLane/Backpack_EarthLane_Cue
root=Modulator_0
nodes [
 {id=Modulator_0, children [Random_1], properties {pitchMax=1, pitchMin=1, volumeMax=0.6, volumeMin=0.5}}
 {id=Random_1, children [WavePlayer_0,WavePlayer_1,WavePlayer_2,WavePlayer_6,WavePlayer_7], properties {weights [1,1,1,1,1]}}
 {id=WavePlayer_0, properties {soundWaveAssetPtr=&/IP_Backpack_EarthLane_ElimReact_01}}
 {id=WavePlayer_1, properties {soundWaveAssetPtr=&/IP_Backpack_EarthLane_ElimReact_02}}
 {id=WavePlayer_2, properties {soundWaveAssetPtr=&/IP_Backpack_EarthLane_ElimReact_03}}
 {id=WavePlayer_6, properties {soundWaveAssetPtr=&/IP_Backpack_EarthLane_ElimReact_04}}
 {id=WavePlayer_7, properties {soundWaveAssetPtr=&/IP_Backpack_EarthLane_ElimReact_05}}
]
properties {
 attenuationSettings=&Fort_Pets_Close_Athena_Att
 audioPropertiesSheet=/BRCosmetics/Sounds/Settings/PropertySheets/Reactive/APS_Cosmetic_Reactive_Backbling
 bOverrideConcurrency=true
 maxDistance=1800.0001
 soundClassObject=/Game/Sounds/FortSoundClassesAndModes/Fort_Backblings
 volumeMultiplier=1
 concurrencyOverrides {maxCount=1, resolutionRule=StopOldest}
}
